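; Parent Blueprint for time-of-day lit props, built on a building actor and carrying a TODLightingComponent.
; The construction script runs the component's construct step with the test day phase for initialization.
; The component is deactivated when dynamic destruction starts, or when the prop's physics object activates on machines that can execute cosmetic events.
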(event EventBeginPlay
  (|CallParentFunction))

(event BuildingActor|PhysicsSimulationSupport|EventHandlePhysicsObjectActivated
  (Networking|CanExecuteCosmeticEvents
    (:True
      (Components|Activation|SetActive (Variables|Default|GetTODLightingComponent) false false))))

(event DynamicDestruction|EventOnDynamicDestructionStarted
  (Components|Activation|SetActive (Variables|Default|GetTODLightingComponent) false false))

(fn ConstructionScript
  (Class|TODLightingComponent|RunConstruct (Variables|Default|GetTODLightingComponent) (Variables|TimeOfDay|GetTest%20Day%20Phase)) ; Initialize the Time of Day component.  
)
