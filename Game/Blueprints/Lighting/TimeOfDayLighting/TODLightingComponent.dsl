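; Time-of-day lighting component that changes its owner's mesh material parameters, mesh visibility and light intensities per day phase.
; RunConstruct matches configured data to components by name, filters invalid entries and converts them to native mappings.
; BeginPlay registers for phase changes unless lighting quality is zero; updates are staggered by location-seeded delays, with a CVar-selected native path.
; High effects quality enables light-function materials with location-seeded random scalars.
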(event Custom|UpdateOnTimeOfDayManagerReady
  (bind _timeofdaymanager (Class|FortAsyncActionTimeOfDayManagerReady|GetTimeOfDayManagerReady)
    (:OnReady
      (Protected|HandleTodPhaseChanged (TimeOfDay|GetCurrentDayPhase _timeofdaymanager :!class /Script/FortniteGame.FortTimeOfDayManagerInterface) true))))

(event EventBeginPlay
  (|CallParentFunction)
  (Networking|CanExecuteCosmeticEvents
    (:True
      (if (Variables|Activation|GetbAutoActivate)
        (EventDispatchers|BindEventtoOnEffectsQualityChanged (Settings|GetFortGameUserSettings) (EventDispatchers|CreateEvent :!event HandleOnEffectsQualityChanged))
        (Protected|InitLightingQuality)
        (if (!= (Variables|Protected|GetLightingQuality) 0)
          (if (> (+ (CallFunction|NumMeshMappings) (CallFunction|NumLightMappings)) 0)
            (Utilities|FlowControl|Sequence
              (:then_0
                (Components|Activation|AssignOnComponentDeactivated (EventDispatchers|CreateEvent :!event Custom|OnComponentDeactivated_Event)))
              (:then_1
                (bind (_asynctaskproxy _timeofdaymanager) (Class|FortAsyncActionWaitforTimeOfDayManagerUpdated|WaitforTimeOfDayManagerUpdated 0)
                  (:then
                    (Variables|FortTimeOfDayComponent|SetAsyncWaitForTimeOfDayManager _asynctaskproxy))
                  (:OnUpdated
                    (CallFunction|NativeUnregisterDayPhaseChanged)
                    (bind _output_get (Variables|FortTimeOfDayComponent|SetCurrentTimeOfDayManager _timeofdaymanager))
                    (TimeOfDay|RegisterforDayPhaseChangedEvent _output_get (EventDispatchers|CreateEvent :!event Custom|HandleDayPhaseChanged) true))))
              (:then_2
                (Components|Activation|BindEventtoOnComponentActivated (EventDispatchers|CreateEvent :!event TODLightingComponent_AutoGenFunc))))))))))

(event Custom|HandleDayPhaseChanged
  (params (CurrentDayPhase :type EFortDayPhase%20Enum :default "Morning") (PreviousDayPhase :type EFortDayPhase%20Enum :default "Morning") (bAtCreation :type Boolean))
  (if (> (Fortnite|GetNumericCVar "Fort.Rollback.UseNativeHandleDayPhaseChangedAmortization") 0)
    (CallFunction|CallHandleDayPhaseChangedAmortized CurrentDayPhase bAtCreation)
    (else
      (Networking|CanExecuteCosmeticEvents
        (:True
          (if (Components|Activation|IsActive)
            (bind _amortizationdelay (Protected|GetAmortizationDelay))
            (Utilities|FlowControl|RetriggerableDelay _amortizationdelay)
            (Protected|HandleTodPhaseChanged CurrentDayPhase bAtCreation)))))))

(event Custom|OnComponentDeactivated_Event
  (params (Component :type Actor%20Component%20Object%20Reference)))

(fn ApplyLightFunction
  (bind _todlightdata (Variables|TimeOfDay|GetTODLightData))
  (bind _lightnamemap (Variables|Protected|GetLightNameMap))
  (if (and (Utilities|Array|IsNotEmpty _todlightdata) (Utilities|Map|IsNotEmpty (Variables|Protected|GetLightNameMap)))
    (bind _returnvalue (Math|Random|MakeRandomStreamfromLocation (Transformation|GetActorLocation (Components|GetOwner)) 1e+01 true))
    (for _array_element _todlightdata
      (bind (_array_element_lightcomponentname_17_6d904f8c4b4a392fed47f7a953377857 _array_element_intensitymultipliers_19_7a5e960d4f135e322711d997fe3d3e13 _array_element_lightfunctionmaterial_23_a0649ee14240ebe75cdbdba8692e1a13) (Utilities|Struct|BreakTODLightingLightDataStruct _array_element))
      (if (and (.ReturnValue (Utilities|Map|Find _lightnamemap _array_element_lightcomponentname_17_6d904f8c4b4a392fed47f7a953377857)) (Utilities|IsValid (Utilities|Map|Find _lightnamemap _array_element_lightcomponentname_17_6d904f8c4b4a392fed47f7a953377857) :!class /Script/Engine.KismetSystemLibrary))
        (bind _returnvalue_2 (Rendering|Material|CreateDynamicMaterialInstance _array_element_lightfunctionmaterial_23_a0649ee14240ebe75cdbdba8692e1a13 "None" "None" :!class /Script/Engine.KismetMaterialLibrary))
        (Utilities|IsValid _returnvalue_2
          (:Is%20Valid
            (Rendering|Material|SetScalarParameterValue _returnvalue_2 "RandomScalar_0-1" (Math|Random|RandomFloatfromStream _returnvalue) :!class /Script/Engine.MaterialInstanceDynamic)
            (Rendering|Components|Light|SetLightFunctionMaterial (Utilities|Map|Find _lightnamemap _array_element_lightcomponentname_17_6d904f8c4b4a392fed47f7a953377857) _returnvalue_2)))))))

(fn ClearLightFunction
  (CallFunction|NativeClearLightFunctions))

(fn ConvertToNativeColors
  (params (Colors :type Day%20Phase%20Linear%20Color%20Structure))
  (bind (_morning_9_da7d290f4da80ac0d79b888f8b1e63af _day_10_7ac9422d4e5da1f7855ecf8df66bafb1 _evening_11_a7e95df64346d1824eb55e8083a4525f _night_12_bd65e285445a3ed05ca26694cf4e873e) (Utilities|Struct|BreakDayPhaseLinearColor Colors))
  (return (Utilities|Struct|MakeTimeOfDayColors _morning_9_da7d290f4da80ac0d79b888f8b1e63af _day_10_7ac9422d4e5da1f7855ecf8df66bafb1 _evening_11_a7e95df64346d1824eb55e8083a4525f _night_12_bd65e285445a3ed05ca26694cf4e873e)))

(fn ConvertToNativeIntensity
  (params (Intensities :type Day%20Phase%20Floats%20Structure))
  (bind (_morning_2_da7d290f4da80ac0d79b888f8b1e63af _day_6_7ac9422d4e5da1f7855ecf8df66bafb1 _evening_7_a7e95df64346d1824eb55e8083a4525f _night_8_bd65e285445a3ed05ca26694cf4e873e) (Utilities|Struct|BreakDayPhaseFloats Intensities))
  (return (Utilities|Struct|MakeTimeOfDayFloats _morning_2_da7d290f4da80ac0d79b888f8b1e63af _day_6_7ac9422d4e5da1f7855ecf8df66bafb1 _evening_7_a7e95df64346d1824eb55e8083a4525f _night_8_bd65e285445a3ed05ca26694cf4e873e)))

(fn ConvertToNativeMaterialData
  (params (MaterialData :type TODLighting%20Material%20Data%20Struct%20Structure))
  (bind (_materialparamname_21_074764624ce1a181bc71c98592aa126e _isafloatparam_22_6c839cdf4ed19c4f278620a0e5f7e523 _floatvalues_18_7a5e960d4f135e322711d997fe3d3e13 _colorvalues_19_cedf2f4740d4efe5c331d0af80fc29b4) (Utilities|Struct|BreakTODLightingMaterialDataStruct MaterialData))
  (return (Utilities|Struct|MakeTimeOfDayMaterialData _materialparamname_21_074764624ce1a181bc71c98592aa126e _isafloatparam_22_6c839cdf4ed19c4f278620a0e5f7e523 (Private|ConverttoNativeIntensity _floatvalues_18_7a5e960d4f135e322711d997fe3d3e13) (Private|ConverttoNativeColors _colorvalues_19_cedf2f4740d4efe5c331d0af80fc29b4))))

(fn ConvertToNativeMaterialIndices
  (params (MaterialIndices :type Array%20of%20TODLighting%20Material%20Data%20Struct%20Structures))
  (for _array_element MaterialIndices
    (Utilities|Array|Add (LocalVariables|Getl_array) (Private|ConverttoNativeMaterialData _array_element)))
  (return (LocalVariables|Getl_array)))

(fn ConvertToNativeMeshData
  (params (MeshData :type TODLighting%20Mesh%20Data%20Struct%20Structure))
  (bind (_meshcomponentname_6_6d904f8c4b4a392fed47f7a953377857 _materialindices_25_cedf2f4740d4efe5c331d0af80fc29b4 _meshvisibility_33_0313daba41b3f638301d8fa5795572cd) (Utilities|Struct|BreakTODLightingMeshDataStruct MeshData))
  (return (Utilities|Struct|MakeTimeOfDayMeshData _meshcomponentname_6_6d904f8c4b4a392fed47f7a953377857 (Private|ConverttoNativeMaterialIndices _materialindices_25_cedf2f4740d4efe5c331d0af80fc29b4) (Private|ConverttoNativeMeshVisibility _meshvisibility_33_0313daba41b3f638301d8fa5795572cd))))

(fn ConvertToNativeMeshVisibility
  (params (MeshVisibility :type Day%20Phase%20Bools%20Structure))
  (bind (_morning_9_da7d290f4da80ac0d79b888f8b1e63af _day_10_7ac9422d4e5da1f7855ecf8df66bafb1 _evening_11_a7e95df64346d1824eb55e8083a4525f _night_12_bd65e285445a3ed05ca26694cf4e873e) (Utilities|Struct|BreakDayPhaseBools MeshVisibility))
  (return (Utilities|Struct|MakeTimeOfDayBools _morning_9_da7d290f4da80ac0d79b888f8b1e63af _day_10_7ac9422d4e5da1f7855ecf8df66bafb1 _evening_11_a7e95df64346d1824eb55e8083a4525f _night_12_bd65e285445a3ed05ca26694cf4e873e)))

(fn GenerateComponentToDataMap
  (bind _lightnamemap (Variables|Protected|GetLightNameMap))
  (CallFunction|ClearMappings)
  (Utilities|FlowControl|Sequence
    (:then_0
      (bind _output_get (LocalVariables|Setl_meshComps (Actor|GetComponentsByClass (Components|GetOwner) "/Script/Engine.MeshComponent")))
      (for _array_element _output_get
        (Utilities|Map|Add (LocalVariables|Getl_meshNameMap) (CallFunction|GetFriendlyCompName _array_element) _array_element))
      (for _array_element_2 (Variables|TimeOfDay|GetTODMeshData)
        (bind (_valid _comp _native_data) (CallFunction|ShouldMeshAddtoMap (LocalVariables|Getl_meshNameMap) _array_element_2))
        (if _valid
          (CallFunction|AddMeshData _comp _native_data))))
    (:then_1
      (bind _output_get_2 (LocalVariables|Setl_lightComps (Actor|GetComponentsByClass (Components|GetOwner) "/Script/Engine.LightComponent")))
      (for _array_element_3 _output_get_2
        (if (CallFunction|ShouldLightAddtoMap _array_element_3)
          (Utilities|Map|Add (Variables|Protected|GetLightNameMap) (CallFunction|GetFriendlyCompName _array_element_3) _array_element_3)))
      (for _array_element_4 (Variables|TimeOfDay|GetTODLightData)
        (if (CallFunction|IsLightDataValid _array_element_4)
          (if (.ReturnValue (Utilities|Map|Find _lightnamemap (Utilities|Struct|BreakTODLightingLightDataStruct _array_element_4)))
            (CallFunction|AddLightData (Utilities|Map|Find _lightnamemap (Utilities|Struct|BreakTODLightingLightDataStruct _array_element_4)) (Utilities|Struct|MakeTimeOfDayLightData (Utilities|Struct|BreakTODLightingLightDataStruct _array_element_4) (Utilities|Struct|MakeTimeOfDayFloats (* (Utilities|Struct|BreakDayPhaseFloats (.IntensityMultipliers_19_7A5E960D4F135E322711D997FE3D3E13 (Utilities|Struct|BreakTODLightingLightDataStruct _array_element_4))) (Variables|/Script/Engine.LightComponentBase|GetIntensity (Utilities|Map|Find _lightnamemap (Utilities|Struct|BreakTODLightingLightDataStruct _array_element_4)))) (* (.Day_6_7AC9422D4E5DA1F7855ECF8DF66BAFB1 (Utilities|Struct|BreakDayPhaseFloats (.IntensityMultipliers_19_7A5E960D4F135E322711D997FE3D3E13 (Utilities|Struct|BreakTODLightingLightDataStruct _array_element_4)))) (Variables|/Script/Engine.LightComponentBase|GetIntensity (Utilities|Map|Find _lightnamemap (Utilities|Struct|BreakTODLightingLightDataStruct _array_element_4)))) (* (.Evening_7_A7E95DF64346D1824EB55E8083A4525F (Utilities|Struct|BreakDayPhaseFloats (.IntensityMultipliers_19_7A5E960D4F135E322711D997FE3D3E13 (Utilities|Struct|BreakTODLightingLightDataStruct _array_element_4)))) (Variables|/Script/Engine.LightComponentBase|GetIntensity (Utilities|Map|Find _lightnamemap (Utilities|Struct|BreakTODLightingLightDataStruct _array_element_4)))) (* (.Night_8_BD65E285445A3ED05CA26694CF4E873E (Utilities|Struct|BreakDayPhaseFloats (.IntensityMultipliers_19_7A5E960D4F135E322711D997FE3D3E13 (Utilities|Struct|BreakTODLightingLightDataStruct _array_element_4)))) (Variables|/Script/Engine.LightComponentBase|GetIntensity (Utilities|Map|Find _lightnamemap (Utilities|Struct|BreakTODLightingLightDataStruct _array_element_4))))) (.LightFunctionMaterial_23_A0649EE14240EBE75CDBDBA8692E1A13 (Utilities|Struct|BreakTODLightingLightDataStruct _array_element_4))))))))))

(fn getAmortizationDelay
  (Utilities|FlowControl|Sequence
    (:then_0
      (if (not (> (Variables|FortTimeOfDayComponent|GetNativeAmortizationDelay) 0))
        (Variables|FortTimeOfDayComponent|SetNativeAmortizationDelay (Utilities|RemapFloat (Class|EnvironmentArtFuncLib|RandomFloatbyLocation (Transformation|GetActorLocation (Components|GetOwner)) 1.0 true) "(X=0.000000,Y=1.000000)" "(X=0.100000,Y=2.000000)" true))))
    (:then_1
      (return (Variables|FortTimeOfDayComponent|GetNativeAmortizationDelay)))))

(fn GetBrightestLightIntensity
  (return (CallFunction|NativeGetBrightestLightIntensity)))

(fn GetFriendlyCompName
  (params (comp :type Object%20Reference))
  (bind _returnvalue_pathstring (Utilities|BreakSoftObjectPath (Utilities|GetSoftObjectPath comp)))
  (bind (_lefts _rights) (Utilities|String|Split _returnvalue_pathstring "." "IgnoreCase" "FromEnd"))
  (return _rights))

(fn GetValueFromDayPhaseFloats
  (params (DayPhaseFloats :type Day%20Phase%20Floats%20Structure%20%28by%20ref%29))
  (bind (_morning_2_da7d290f4da80ac0d79b888f8b1e63af _day_6_7ac9422d4e5da1f7855ecf8df66bafb1 _evening_7_a7e95df64346d1824eb55e8083a4525f _night_8_bd65e285445a3ed05ca26694cf4e873e) (Utilities|Struct|BreakDayPhaseFloats DayPhaseFloats))
  (return (Utilities|Select :Index (Variables|FortTimeOfDayComponent|GetNativeDayPhase) :Morning _morning_2_da7d290f4da80ac0d79b888f8b1e63af :Day _day_6_7ac9422d4e5da1f7855ecf8df66bafb1 :Evening _evening_7_a7e95df64346d1824eb55e8083a4525f :Night _night_8_bd65e285445a3ed05ca26694cf4e873e)))

(fn GetVectorFromDayPhaseColor
  (params (DayPhaseLinearColor :type Day%20Phase%20Linear%20Color%20Structure%20%28by%20ref%29))
  (bind (_morning_9_da7d290f4da80ac0d79b888f8b1e63af _day_10_7ac9422d4e5da1f7855ecf8df66bafb1 _evening_11_a7e95df64346d1824eb55e8083a4525f _night_12_bd65e285445a3ed05ca26694cf4e873e) (Utilities|Struct|BreakDayPhaseLinearColor DayPhaseLinearColor))
  (return (Math|Conversions|ToVector(LinearColor) (Utilities|Select :Index (Variables|FortTimeOfDayComponent|GetNativeDayPhase) :Morning _morning_9_da7d290f4da80ac0d79b888f8b1e63af :Day _day_10_7ac9422d4e5da1f7855ecf8df66bafb1 :Evening _evening_11_a7e95df64346d1824eb55e8083a4525f :Night _night_12_bd65e285445a3ed05ca26694cf4e873e))))

(fn GetVisibleFromDayPhaseBools
  (params (DayPhaseLinearColor :type Day%20Phase%20Bools%20Structure%20%28by%20ref%29))
  (bind (_morning_9_da7d290f4da80ac0d79b888f8b1e63af _day_10_7ac9422d4e5da1f7855ecf8df66bafb1 _evening_11_a7e95df64346d1824eb55e8083a4525f _night_12_bd65e285445a3ed05ca26694cf4e873e) (Utilities|Struct|BreakDayPhaseBools DayPhaseLinearColor))
  (return (Utilities|Select :Index (Variables|FortTimeOfDayComponent|GetNativeDayPhase) :Morning _morning_9_da7d290f4da80ac0d79b888f8b1e63af :Day _day_10_7ac9422d4e5da1f7855ecf8df66bafb1 :Evening _evening_11_a7e95df64346d1824eb55e8083a4525f :Night _night_12_bd65e285445a3ed05ca26694cf4e873e)))

(fn HandleOnEffectsQualityChanged
  (if (CallFunction|IsLightFunctionAllowed)
    (Protected|ApplyLightFunction)
    (else
      (CallFunction|ClearLightFunction))))

(fn HandleTodPhaseChanged
  (params (currentDayPhase :type EFortDayPhase%20Enum :default "Morning") (ForceUpdate :type Boolean))
  (CallFunction|NativeHandleDayPhaseChanged currentDayPhase ForceUpdate))

(fn HasLightData
  (if (> (CallFunction|NumLightMappings) 0)
    (return true)
    (else
      (return false))))

(fn InitLightingQuality
  (if (or (TouchPlatform|IsMobileGame) (Class|PlatformHelpersLibrary|IsPcPerfMode))
    (Variables|Protected|SetLightingQuality 0)
    (elif (TouchPlatform|IsNintendoSwitchorNintendoSwitchPreview)
      (Variables|Protected|SetLightingQuality 1)
      (else
        (Variables|Protected|SetLightingQuality (Rendering|GetRenderingDetailMode))))))

(fn IsLightDataValid
  (params (LightData :type TODLighting%20Light%20Data%20Struct%20Structure%20%28by%20ref%29))
  (bind (_lightcomponentname_17_6d904f8c4b4a392fed47f7a953377857 _intensitymultipliers_19_7a5e960d4f135e322711d997fe3d3e13) (Utilities|Struct|BreakTODLightingLightDataStruct LightData))
  (bind (_morning_2_da7d290f4da80ac0d79b888f8b1e63af _day_6_7ac9422d4e5da1f7855ecf8df66bafb1 _evening_7_a7e95df64346d1824eb55e8083a4525f _night_8_bd65e285445a3ed05ca26694cf4e873e) (Utilities|Struct|BreakDayPhaseFloats _intensitymultipliers_19_7a5e960d4f135e322711d997fe3d3e13))
  (return (!= (Utilities|Operators|Add (Math|Float|Max(Float) _morning_2_da7d290f4da80ac0d79b888f8b1e63af) (Math|Float|Max(Float) _day_6_7ac9422d4e5da1f7855ecf8df66bafb1) (Math|Float|Max(Float) _evening_7_a7e95df64346d1824eb55e8083a4525f) (Math|Float|Max(Float) _night_8_bd65e285445a3ed05ca26694cf4e873e)) 0)))

(fn IsLightFunctionAllowed
  (return (> (Settings|GetVisualEffectQuality (Settings|GetFortGameUserSettings)) 2)))

(fn RunConstruct
  (params (TestDayPhase :type EFortDayPhase%20Enum :default "Morning"))
  (Variables|Protected|SetTestDayPhase TestDayPhase)
  (if (> (+ (Utilities|Array|Length (Variables|TimeOfDay|GetTODMeshData)) (Utilities|Array|Length (Variables|TimeOfDay|GetTODLightData))) 0)
    (Utilities|FlowControl|Sequence
      (:then_0
        (if (> (Fortnite|GetNumericCVar "Fort.Rollback.UseNativeHandleDayPhaseChangedAmortization") 0)
          (CallFunction|NativeGetAmortizationDelay)
          (else
            (Protected|GetAmortizationDelay) ; This stores a value on construct so we don't run a rand during gameplay
          )))
      (:then_1
        (Protected|GenerateComponenttoDataMap))
      (:then_2
        (if (Editor|InEditor)
          (Protected|HandleTodPhaseChanged (Variables|Protected|GetTestDayPhase) true)
          (else
            (CallFunction|UpdateonTimeOfDayManagerReady))))
      (:then_3
        (if (CallFunction|IsLightFunctionAllowed)
          (Protected|ApplyLightFunction))))))

(fn ShouldLightAddToMap
  (params (light :type Light%20Component%20Base%20Object%20Reference))
  (return (Math|Boolean|ANDBoolean (Variables|/Script/Engine.SceneComponent|GetbVisible light) (> (Variables|/Script/Engine.LightComponentBase|GetIntensity light) 0) (Variables|/Script/Engine.LightComponentBase|GetbAffectsWorld light) (not (Variables|/Script/Engine.SceneComponent|GetbHiddenInGame light)))))

(fn ShouldMeshAddToMap
  (params (TargetMap :type Map%20of%20Strings%20to%20Mesh%20Component%20Object%20References) (TODLighting_MeshDataStruct :type TODLighting%20Mesh%20Data%20Struct%20Structure%20%28by%20ref%29))
  (bind (_meshcomponentname_6_6d904f8c4b4a392fed47f7a953377857 _materialindices_25_cedf2f4740d4efe5c331d0af80fc29b4) (Utilities|Struct|BreakTODLightingMeshDataStruct (LocalVariables|Getl_struct)))
  (bind (_meshcomponentname_6_6d904f8c4b4a392fed47f7a953377857_2 _materialindices_25_cedf2f4740d4efe5c331d0af80fc29b4_2 _meshvisibility_33_0313daba41b3f638301d8fa5795572cd) (Utilities|Struct|BreakTODLightingMeshDataStruct (LocalVariables|Getl_struct)))
  (bind (_morning_9_da7d290f4da80ac0d79b888f8b1e63af _day_10_7ac9422d4e5da1f7855ecf8df66bafb1 _evening_11_a7e95df64346d1824eb55e8083a4525f _night_12_bd65e285445a3ed05ca26694cf4e873e) (Utilities|Struct|BreakDayPhaseBools _meshvisibility_33_0313daba41b3f638301d8fa5795572cd))
  (bind (_materialparamname_21_074764624ce1a181bc71c98592aa126e _isafloatparam_22_6c839cdf4ed19c4f278620a0e5f7e523) (Utilities|Struct|BreakTODLightingMaterialDataStruct (LocalVariables|Getl_matStruct)))
  (bind _l_finalmaterialindexdata (LocalVariables|Getl_finalMaterialIndexData))
  (LocalVariables|Setl_struct TODLighting_MeshDataStruct)
  (bind _output_get (LocalVariables|Setl_foundComp (Utilities|Map|Find TargetMap (Utilities|Struct|BreakTODLightingMeshDataStruct (LocalVariables|Getl_struct)))))
  (Utilities|IsValid _output_get
    (:Is%20Valid
      (Utilities|FlowControl|Sequence
        (:then_0
          (for (_array_element _array_index) _materialindices_25_cedf2f4740d4efe5c331d0af80fc29b4
            (LocalVariables|Setl_matStruct _array_element)
            (bind _asmaterial_instance_constant (Utilities|GetBaseMaterialInstance (Rendering|Material|GetMaterial (LocalVariables|Getl_foundComp) _array_index))
              (:Valid
                (LocalVariables|Setl_foundIt false)
                (Utilities|FlowControl|Sequence
                  (:then_0
                    (if _isafloatparam_22_6c839cdf4ed19c4f278620a0e5f7e523
                      (bind _array_element_2 (Utilities|Array|ForEachLoopwithBreak (Variables|/Script/Engine.MaterialInstance|GetScalarParameterValues _asmaterial_instance_constant) :!label _label)
                        (:LoopBody
                          (bind _parameterinfo (Utilities|Struct|BreakScalarParameterValue _array_element_2))
                          (bind _parameterinfo_name (Utilities|Struct|BreakMaterialParameterInfo _parameterinfo))
                          (if (== _parameterinfo_name (Utilities|Struct|BreakTODLightingMaterialDataStruct (LocalVariables|Getl_matStruct)))
                            (LocalVariables|Setl_foundIt true)
                            (LocalVariables|Setl_atLeastOneValidMaterial true)
                            (goto _label Break))))
                      (else
                        (bind _array_element_3 (Utilities|Array|ForEachLoopwithBreak (Variables|/Script/Engine.MaterialInstance|GetVectorParameterValues _asmaterial_instance_constant) :!label _label_2)
                          (:LoopBody
                            (bind _parameterinfo_2 (Utilities|Struct|BreakVectorParameterValue _array_element_3))
                            (bind _parameterinfo_name_2 (Utilities|Struct|BreakMaterialParameterInfo _parameterinfo_2))
                            (if (== _parameterinfo_name_2 (Utilities|Struct|BreakTODLightingMaterialDataStruct (LocalVariables|Getl_matStruct)))
                              (LocalVariables|Setl_foundIt true)
                              (LocalVariables|Setl_atLeastOneValidMaterial true)
                              (goto _label_2 Break)))))))
                  (:then_1
                    (if (LocalVariables|Getl_foundIt)
                      (Utilities|Array|Add _l_finalmaterialindexdata (LocalVariables|Getl_matStruct))
                      (else
                        (Utilities|Array|Add _l_finalmaterialindexdata (Utilities|Struct|MakeTODLightingMaterialDataStruct "None")))))))
              (:NotMIC
                (LocalVariables|Setl_foundIt true)
                (LocalVariables|Setl_atLeastOneValidMaterial true) ; Directly uses a material, just assume it worked (we can't query without making MIDs)
              ))))
        (:then_1
          (if (LocalVariables|Getl_atLeastOneValidMaterial)
            (bind _structout (Utilities|Struct|SetmembersinTODLightingMeshDataStruct (LocalVariables|Getl_struct) (LocalVariables|Getl_finalMaterialIndexData))) ; We have some valid params to set
            (return true (LocalVariables|Getl_foundComp) (Private|ConverttoNativeMeshData _structout))
            (elif (Math|Boolean|ANDBoolean (== _morning_9_da7d290f4da80ac0d79b888f8b1e63af _day_10_7ac9422d4e5da1f7855ecf8df66bafb1) (== _day_10_7ac9422d4e5da1f7855ecf8df66bafb1 _evening_11_a7e95df64346d1824eb55e8083a4525f) (== _evening_11_a7e95df64346d1824eb55e8083a4525f _night_12_bd65e285445a3ed05ca26694cf4e873e))
              (return false 0 0)
              (else
                (bind _structout_2 (Utilities|Struct|SetmembersinTODLightingMeshDataStruct (LocalVariables|Getl_struct) (Utilities|Array|MakeArray))) ; No Valid material params to set but we do need to set visibility
                (return true (LocalVariables|Getl_foundComp) (Private|ConverttoNativeMeshData _structout_2))))))))
    (:Is%20Not%20Valid
      (return false 0 0))))

(fn TODLightingComponent_AutoGenFunc
  (params (Component :type Actor%20Component%20Object%20Reference) (bReset :type Boolean))
  (CallFunction|HandleDayPhaseChanged (TimeOfDay|GetCurrentDayPhase (Variables|FortTimeOfDayComponent|GetCurrentTimeOfDayManager) :!class /Script/FortniteGame.FortTimeOfDayManagerInterface) "Morning" true))

(fn TryGetBaseMaterialInstance
  (params (MatInterface :type Material%20Interface%20Object%20Reference))
  (bind _asmaterial_instance_constant (Utilities|Casting|CastToMaterialInstanceConstant MatInterface)
    (:then
      (return true _asmaterial_instance_constant))
    (:CastFailed
      (bind _asmaterial_instance_dynamic (Utilities|Casting|CastToMaterialInstanceDynamic MatInterface))
      (bind _asmaterial_instance_constant_2 (Utilities|Casting|CastToMaterialInstanceConstant (Variables|/Script/Engine.MaterialInstance|GetParent _asmaterial_instance_dynamic))) ; Just try once, recursion is for suckers\n(Odds of stacked MIDs is low)
      (return true _asmaterial_instance_constant_2))))
